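# Physics asset for the face accessory part of the female medium TealMink Pelt Helmet cosmetic, setting up physics for its dangling dreadlock bones.
# Capsules cover two front and three back dread groups, plus head, neck, upper spine, clavicles and arms as body collision shapes.
# Swing and twist limited constraints chain each dread from its base, with branches in the back groups.
# Free-angular cross-constraints tie neighboring back strands together without restricting their relative rotation.
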PhysicsAsset /BRCosmetics/Characters/Player/Female/Medium/Bodies/F_MED_TealMink_Pelt_Helmet/Meshes/Parts/F_MED_TealMink_Pelt_Helmet_FaceAcc_Physics
bodies {
 clavicle_l [{
   shapeType=Capsule
   length=8.255
   radius=5.755
   shapeName=clavicle_l_capsule
   center=<5.4681 -1.0669 -2.0407>
   rotation=<rot -0.6857 -84.9909 94.9656>
  }]
 clavicle_r [{
   shapeType=Capsule
   length=8.505
   radius=5.505
   shapeName=clavicle_r_capsule
   center=<-5.6416 1.1527 1.606>
   rotation=<rot -0.2509 -89.9721 100.0063>
  }]
 dyn_dread_back_1_1 [{
   shapeType=Capsule
   length=7.2476
   radius=2.0242
   shapeName=dyn_dread_back_1_1_capsule
   center=<8.3153 0.4654 0.7778>
   rotation=<rot -19.0933 -112.9485 -77.5228>
  }]
 dyn_dread_back_1_3 [{
   shapeType=Capsule
   length=3.4414
   radius=1.8831
   shapeName=dyn_dread_back_1_3_capsule
   center=<6.9576 -0.3772 0.3322>
   rotation=<rot 0.0002 -102.8078 -93.4278>
  }]
 dyn_dread_back_1_4 [{
   shapeType=Capsule
   length=6.619
   radius=1.2402
   shapeName=dyn_dread_back_1_4_capsule
   center=<5.6825 -0.0915 0.3228>
   rotation=<rot 0 90.0499 -87.0794>
  }]
 dyn_dread_back_1_5 [{
   shapeType=Capsule
   length=6.0002
   radius=0.75
   shapeName=dyn_dread_back_1_5_capsule
   center=<4.6892 0.1191 -0.0345>
   rotation=<rot -9.9984 90.0211 -91.0428>
  }]
 dyn_dread_back_1_6 [{
   shapeType=Capsule
   length=3.2773
   radius=1.1043
   shapeName=dyn_dread_back_1_6_capsule
   center=<3.1179 0.0708 -0.0201>
   rotation=<rot 0 -88.9376 -89.8894>
  }]
 dyn_dread_back_1_base [{shapeType=Capsule, length=0, radius=2, shapeName=dyn_dread_back_1_base_capsule, rotation=<rot 0 -90 0>}]
 dyn_dread_back_2_1 [{
   shapeType=Capsule
   length=3
   radius=2.005
   shapeName=dyn_dread_back_2_1_capsule
   center=<5.9251 -0.0042 0.7549>
   rotation=<rot -5 -179.9996 180>
  }]
 dyn_dread_back_2_2 [{
   shapeType=Capsule
   length=3.846
   radius=1.8204
   shapeName=dyn_dread_back_2_2_capsule
   center=<3.5281 0.2001 -0.0153>
   rotation=<rot -0.1341 -107.0106 -90.5005>
  }]
 dyn_dread_back_2_3 [{
   shapeType=Capsule
   length=6.3507
   radius=0.7742
   shapeName=dyn_dread_back_2_3_capsule
   center=<3.8615 0.154 -0.1812>
   rotation=<rot -0.1719 86.0615 -88.0349>
  }]
 dyn_dread_back_2_4 [{
   shapeType=Capsule
   length=5.8692
   radius=0.5749
   shapeName=dyn_dread_back_2_4_capsule
   center=<4.0958 0.2688 0.1767>
   rotation=<rot 0.4693 84.6109 -87.6625>
  }]
 dyn_dread_back_2_5 [{
   shapeType=Capsule
   length=4.3372
   radius=0.8079
   shapeName=dyn_dread_back_2_5_capsule
   center=<2.4947 0.0404 0.2223>
   rotation=<rot 0.0561 -90.535 -89.3598>
  }]
 dyn_dread_back_2_6 [{
   shapeType=Capsule
   length=4.6855
   radius=0.9757
   shapeName=dyn_dread_back_2_6_capsule
   center=<3.5424 -0.4429 0.1214>
   rotation=<rot 0 89.9647 -87.5333>
  }]
 dyn_dread_back_2_base [{shapeType=Capsule, length=0, radius=1.5, shapeName=dyn_dread_back_2_base_capsule, rotation=<rot 0 -90 0>}]
 dyn_dread_back_3_1 [{
   shapeType=Capsule
   length=5.5888
   radius=1.3589
   shapeName=dyn_dread_back_3_1_capsule
   center=<5.3154 0.6428 -0.0721>
   rotation=<rot -7.5203 65.0424 -72.7294>
  }]
 dyn_dread_back_3_2 [{
   shapeType=Capsule
   length=2.1191
   radius=1.5945
   shapeName=dyn_dread_back_3_2_capsule
   center=<2.8419 -0.5529 -0.029>
   rotation=<rot -9.7963 74.7355 -83.3318>
  }]
 dyn_dread_back_3_3 [{
   shapeType=Capsule
   length=4.7511
   radius=1.5808
   shapeName=dyn_dread_back_3_3_capsule
   center=<2.7184 -0.3597 -0.2844>
   rotation=<rot 14.9988 -83.2778 -89.2668>
  }]
 dyn_dread_back_3_4 [{
   shapeType=Capsule
   length=6.0018
   radius=1.0554
   shapeName=dyn_dread_back_3_4_capsule
   center=<3.9062 -0.5955 0.176>
   rotation=<rot 0 93.0774 -93.0024>
  }]
 dyn_dread_back_3_5 [{
   shapeType=Capsule
   length=5.0788
   radius=0.8971
   shapeName=dyn_dread_back_3_5_capsule
   center=<3.4128 0.4972 -0.1112>
   rotation=<rot -0.0124 -83.5062 -89.8588>
  }]
 dyn_dread_back_3_base [{shapeType=Capsule, length=0, radius=1.5, shapeName=dyn_dread_back_3_base_capsule, rotation=<rot 0 -90 0>}]
 dyn_dread_front_1_1_l [{
   shapeType=Capsule
   length=6.211
   radius=1.6751
   shapeName=dyn_dread_front_1_1_l_capsule
   center=<7.4131 0.4145 1.1883>
   rotation=<rot -3.5118 -5.8799 -72.5849>
  }]
 dyn_dread_front_1_1_r [{
   shapeType=Capsule
   length=5.708
   radius=1.6209
   shapeName=dyn_dread_front_1_1_r_capsule
   center=<6.3446 1.0637 -0.4513>
   rotation=<rot -16.4307 -146.9598 -59.6478>
  }]
 dyn_dread_front_1_2_l [{
   shapeType=Capsule
   length=5.0812
   radius=1.7294
   shapeName=dyn_dread_front_1_2_l_capsule
   center=<3.5815 1.1225 0.2832>
   rotation=<rot -2.1098 68.1602 -89.4395>
  }]
 dyn_dread_front_1_2_r [{
   shapeType=Capsule
   length=4.9991
   radius=1.7346
   shapeName=dyn_dread_front_1_2_r_capsule
   center=<3.7906 0.4047 -1.4672>
   rotation=<rot -45.7426 -100.5999 -111.9735>
  }]
 dyn_dread_front_1_3_l [{
   shapeType=Capsule
   length=5.0669
   radius=1.25
   shapeName=dyn_dread_front_1_3_l_capsule
   center=<4.1522 -1.3489 0.1223>
   rotation=<rot 0.1364 66.9442 -88.4413>
  }]
 dyn_dread_front_1_3_r [{
   shapeType=Capsule
   length=7.9264
   radius=1.9853
   shapeName=dyn_dread_front_1_3_r_capsule
   center=<5.1588 -1.6237 1.3335>
   rotation=<rot 0 70.7377 -77.1987>
  }]
 dyn_dread_front_1_base_l [{shapeType=Capsule, length=0.505, radius=1.505, shapeName=dyn_dread_front_1_base_l_capsule, rotation=<rot 0 -90 0>}]
 dyn_dread_front_1_base_r [{shapeType=Capsule, length=0, radius=1.5, shapeName=dyn_dread_front_1_base_r_capsule, rotation=<rot 0 -90 0>}]
 dyn_head_corr [
  {
   shapeType=Capsule
   length=8.255
   radius=4.755
   shapeName=dyn_head_corr_capsule
   center=<2.46 -7.5027 -5.9733>
   rotation=<rot -45 -90.0003 15>
  }
  {
   shapeType=Capsule
   length=7.25
   radius=4.5
   shapeName=dyn_head_corr_capsule
   center=<2.7209 -8.0277 6.014>
   rotation=<rot 7.6443 6.4661 40.4325>
  }
 ]
 head [
  {
   shapeType=Capsule
   length=6.5794
   radius=4.2844
   shapeName=head_capsule
   center=<1.144 -0.6162 0.1747>
   rotation=<rot -0.5497 -90.0007 -87.4122>
  }
  {shapeType=Capsule, length=12, radius=2, shapeName=head_capsule, center=<-5.8972 4.2016 -9.4567>, rotation=<rot 4.9809 29.6211 -103.6822>}
 ]
 lowerarm_l [{
   shapeType=Capsule
   length=17.505
   radius=4.005
   shapeName=lowerarm_l_capsule
   center=<11.8676 2.6616 0.0875>
   rotation=<rot 0.4311 -80.0179 90.0562>
  }]
 lowerarm_r [{
   shapeType=Capsule
   length=17.505
   radius=4.005
   shapeName=lowerarm_r_capsule
   center=<-11.9659 -2.1502 0.105>
   rotation=<rot 0.4311 -80.0179 90.0561>
  }]
 neck_01 [{shapeType=Capsule, length=5.5, radius=4.5, shapeName=neck_01_capsule, center=<0 1.1266 0>, rotation=<rot 0 -90.0002 90>}]
 spine_04 [{shapeType=Capsule, length=6.5, radius=10.75, shapeName=spine_04_capsule, center=<-1.0066 1.4529 0>, rotation=<rot 0 -90 0>}]
 spine_05 [{shapeType=Capsule, length=16, radius=7.5, shapeName=spine_05_capsule, center=<1.7815 -1.0086 0>, rotation=<rot 0 -90 0>}]
 upperarm_l [{shapeType=Capsule, length=19.005, radius=3.505, shapeName=upperarm_l_capsule, center=<14.0596 -0.7318 0>, rotation=<rot -0.0003 -94.9997 90>}]
 upperarm_r [{shapeType=Capsule, length=19.005, radius=3.505, shapeName=upperarm_r_capsule, center=<-13.5909 1.1482 0>, rotation=<rot -0.0004 -89.9996 90>}]
}
constraints [
 {bone1Name=dyn_dread_front_1_1_r, bone2Name=dyn_dread_front_1_base_r, swing1LimitDegrees=30, swing2LimitDegrees=20, twistLimitDegrees=30}
 {bone1Name=dyn_dread_front_1_2_r, bone2Name=dyn_dread_front_1_1_r, swing1LimitDegrees=40, swing2LimitDegrees=30, twistLimitDegrees=30}
 {bone1Name=dyn_dread_front_1_3_r, bone2Name=dyn_dread_front_1_2_r, swing1LimitDegrees=30, swing2LimitDegrees=30, twistLimitDegrees=10}
 {bone1Name=dyn_dread_front_1_1_l, bone2Name=dyn_dread_front_1_base_l, swing1LimitDegrees=20, swing2LimitDegrees=20, twistLimitDegrees=10}
 {bone1Name=dyn_dread_front_1_3_l, bone2Name=dyn_dread_front_1_2_l, swing1LimitDegrees=40, swing2LimitDegrees=20, twistLimitDegrees=20}
 {bone1Name=dyn_dread_back_3_1, bone2Name=dyn_dread_back_3_base, swing1LimitDegrees=60, swing2LimitDegrees=30, twistLimitDegrees=20}
 {bone1Name=dyn_dread_back_3_3, bone2Name=dyn_dread_back_3_2, swing1LimitDegrees=30, swing2LimitDegrees=15, twistLimitDegrees=20}
 {bone1Name=dyn_dread_back_3_5, bone2Name=dyn_dread_back_3_4, swing1LimitDegrees=25, swing2LimitDegrees=20, twistLimitDegrees=30}
 {bone1Name=dyn_dread_back_2_1, bone2Name=dyn_dread_back_2_base, swing1LimitDegrees=40, swing2LimitDegrees=45, twistLimitDegrees=30}
 {bone1Name=dyn_dread_back_2_6, bone2Name=dyn_dread_back_2_5, swing1LimitDegrees=30, swing2LimitDegrees=35, twistLimitDegrees=35}
 {bone1Name=dyn_dread_back_1_1, bone2Name=dyn_dread_back_1_base, swing1LimitDegrees=35, swing2LimitDegrees=25, twistLimitDegrees=30}
 {bone1Name=dyn_dread_back_1_4, bone2Name=dyn_dread_back_1_3, swing1LimitDegrees=30, swing2LimitDegrees=20, twistLimitDegrees=10}
 {bone1Name=dyn_dread_back_1_5, bone2Name=dyn_dread_back_1_3, swing1LimitDegrees=30, swing2LimitDegrees=20, twistLimitDegrees=30}
 {bone1Name=dyn_dread_back_1_6, bone2Name=dyn_dread_back_1_5, swing1LimitDegrees=30, swing2LimitDegrees=20, twistLimitDegrees=30}
 {bone1Name=dyn_dread_back_3_2, bone2Name=dyn_dread_back_3_1, swing1LimitDegrees=60, swing2LimitDegrees=25, twistLimitDegrees=30}
 {bone1Name=dyn_dread_back_3_4, bone2Name=dyn_dread_back_3_1, swing1LimitDegrees=45, swing2LimitDegrees=20, twistLimitDegrees=30}
 {bone1Name=dyn_dread_back_1_3, bone2Name=dyn_dread_back_1_1, swing1LimitDegrees=45, swing2LimitDegrees=30, twistLimitDegrees=20}
 {bone1Name=dyn_dread_back_2_2, bone2Name=dyn_dread_back_2_1, swing1LimitDegrees=45, swing2LimitDegrees=30, twistLimitDegrees=30}
 {bone1Name=dyn_dread_back_2_3, bone2Name=dyn_dread_back_2_2, swing1LimitDegrees=30, swing2LimitDegrees=30, twistLimitDegrees=30}
 {bone1Name=dyn_dread_back_2_4, bone2Name=dyn_dread_back_2_2, swing1LimitDegrees=30, swing2LimitDegrees=30, twistLimitDegrees=45}
 {bone1Name=dyn_dread_back_2_5, bone2Name=dyn_dread_back_2_2, swing1LimitDegrees=30, swing2LimitDegrees=30, twistLimitDegrees=30}
 {
  bone1Name=dyn_dread_back_3_4
  bone2Name=dyn_dread_back_2_2
  swing1LimitDegrees=45
  swing1Motion=Free
  swing2LimitDegrees=45
  swing2Motion=Free
  twistLimitDegrees=45
  twistMotion=Free
 }
 {
  bone1Name=dyn_dread_back_1_3
  bone2Name=dyn_dread_back_2_2
  swing1LimitDegrees=30
  swing1Motion=Free
  swing2LimitDegrees=30
  swing2Motion=Free
  twistLimitDegrees=20
  twistMotion=Free
 }
 {
  bone1Name=dyn_dread_back_3_4
  bone2Name=dyn_dread_back_3_2
  swing1LimitDegrees=15
  swing2LimitDegrees=15
  twistLimitDegrees=30
  twistMotion=Locked
 }
 {
  bone1Name=dyn_dread_back_3_5
  bone2Name=dyn_dread_back_2_4
  swing1LimitDegrees=45
  swing1Motion=Free
  swing2LimitDegrees=45
  swing2Motion=Free
  twistLimitDegrees=45
  twistMotion=Free
 }
 {bone1Name=dyn_dread_front_1_2_l, bone2Name=dyn_dread_front_1_1_l, swing1LimitDegrees=45, swing2LimitDegrees=45, twistLimitDegrees=20}
 {
  bone1Name=dyn_dread_back_1_5
  bone2Name=dyn_dread_back_2_3
  swing1LimitDegrees=20
  swing1Motion=Free
  swing2LimitDegrees=20
  swing2Motion=Free
  twistLimitDegrees=45
  twistMotion=Free
 }
 {
  bone1Name=dyn_dread_back_3_5
  bone2Name=dyn_dread_back_3_3
  swing1LimitDegrees=45
  swing1Motion=Free
  swing2LimitDegrees=45
  swing2Motion=Free
  twistLimitDegrees=45
  twistMotion=Free
 }
 {
  bone1Name=dyn_dread_back_1_4
  bone2Name=dyn_dread_back_1_5
  swing1LimitDegrees=45
  swing1Motion=Free
  swing2LimitDegrees=45
  swing2Motion=Free
  twistLimitDegrees=45
  twistMotion=Free
 }
]
